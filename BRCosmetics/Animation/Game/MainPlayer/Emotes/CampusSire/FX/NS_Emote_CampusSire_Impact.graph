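# Niagara impact burst for the CampusSire emote, a short blue and pink hit effect.
# Debris (sprites and meshes) and fast Sparks burst in two spawn groups, each with its own color and opposing cone direction, under gravity and drag.
# Flash adds sprites and a brief light, while Impact_Streaks adds short stretching streaks.
# Debris switches between two collision setups by quality level and User.bIsFrontEndPreview; skeletal mesh placement modules are disabled.
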NiagaraSystem /BRCosmetics/Animation/Game/MainPlayer/Emotes/CampusSire/FX/NS_Emote_CampusSire_Impact
dependencies {
 usedDataInterfaces [
  &NiagaraDataInterfaceCurve
  &NiagaraDataInterfaceSpriteRendererInfo
  &NiagaraDataInterfaceMeshRendererInfo
  &NiagaraDataInterfacePlatformSet
  &NiagaraDataInterfaceArrayColor
  &NiagaraDataInterfaceColorCurve
  &NiagaraDataInterfaceVector2DCurve
  &NiagaraDataInterfaceSkeletalMesh
 ]
 usedDynamicInputs []
 usedModules [
  &self:SystemSpawnScript.SystemScriptSource.SystemScriptGraph.NiagaraNodeAssignment_0.SetVariables_#1
  &SystemState
 ]
}
emitters {
 Debris {
  emitterUpdateScript [
   &EmitterState()
   &SpawnBurst_Instantaneous(
    Spawn Count=&RandomRangeInt(Minimum=2, Maximum=5)
    Use Spawn Probability=false
    UseLoopCountLimit=false
   ) {module=SpawnBurst_Instantaneous001}
   &SpawnBurst_Instantaneous(
    Spawn Count=&RandomRangeInt(Minimum=2, Maximum=5)
    Spawn Group=1
    Use Spawn Probability=false
    UseLoopCountLimit=false
   )
  ]
  particleSpawnScript [
   &InitializeParticle(
    Color=&MaskLinearColorBySpawnGroup(
     Masked Linear Color=&RandomRangeLinearColor(Minimum=<rgba 0.2753 0.6621 1 1>, Maximum=<rgba 0 0.4979 1 1>)
     Passthrough Linear Color=&RandomRangeLinearColor(Minimum=<rgba 0.6479 0.1044 1 1>, Maximum=<rgba 0.8716 0.1772 1 1>)
    )
    Lifetime Max=1
    Lifetime Min=0.6
    Lifetime Mode=::NewEnumerator1
    Mesh Renderer Array Visibility Mode=::NewEnumerator3
    Mesh Renderer Info=DataInterface() {propertyValues {MeshRenderer=&self:Debris_0.NiagaraMeshRendererProperties_0}}
    Mesh Scale Mode=::NewEnumerator1
    Mesh Uniform Scale Max=0.5
    Mesh Uniform Scale Min=0.2
    Sprite Rotation Mode=::NewEnumerator2
    Sprite Size Mode=::NewEnumerator1
    Uniform Sprite Size Max=2
    Uniform Sprite Size Min=0.5
    UsePositionOffset=false
    Write Material Random=true
   )
   &ShapeLocation(
    MaskSpawnGroup=false
    Offset=&MaskVectorBySpawnGroup(Masked Vector=<5 0 0>, Passthrough Vector=<-5 0 0>)
    Offset Mode=::NewEnumerator0
    Override Seed=false
    Sphere Radius=15
    UseOwnerScale=true
   )
   &SkeletalMeshLocation(
    Bone / Socket Index=&SpawnGroup()
    Bone Sampling Mode=::NewEnumerator4
    MaskSpawnGroup=false
    Sampled Position Offset=&AddVectorToPosition(
     Position=$Particles.ShapeLocation.ShapePosition
     Vector=&MaskVectorBySpawnGroup(Masked Vector=<5 0 0>, Passthrough Vector=<-5 0 0>)
    )
    Sampled Position Offset Coordinate Space:ENiagaraSkelMeshTransforms=::NewEnumerator2
    Skeletal Mesh=$System.Skeletal Mesh
   ) {enabled=false}
   &AddVelocity(
    Cone Angle=180
    Cone Axis=&MaskVectorBySpawnGroup(Masked Vector=<-1 0.5 0>, Passthrough Vector=<1 0.5 0>)
    Module.UseConeFalloff=true
    Module.UseVelDistribution=true
    Override Seed=false
    Speed Falloff From Cone Axis=0.2
    Velocity Mode=::NewEnumerator2
    Velocity Speed=&RandomRangeFloat(Minimum=125, Maximum=275)
   )
   &SubUVAnimation(
    "Sprite Renderer "=DataInterface() {propertyValues {SpriteRenderer=&self:Debris_0.NiagaraSpriteRendererProperties_0}}
    SubUV Animation Mode:ENiagara_SubUVLookupModeV2=::NewEnumerator2
    UseEndFrame=false
    UseStartFrame=false
   )
   &InitialMeshOrientation(Use Rotation Vector=false)
  ]
  particleUpdateScript [
   &ParticleState()
   &GravityForce(Gravity=&MakeVector(X=0, Y=0, Z=-980))
   &MeshRotationForce(Rotation=&VectorFromFloat(Value=&RandomRangeFloat(Minimum=-90, Maximum=90)))
   &SpriteRotationRate(Rotation Rate=&RandomRangeFloat(Minimum=-40, Maximum=40))
   &Collision(
    Age Colliding Particles=false
    Collision Enabled=&Boolean_AndOperation(
     Bool1=&IsPlatformSetActive(New Platform Set=DataInterface() {propertyValues {Platforms {qualityLevelMask=-8}}})
     Bool2=&Boolean_NotOperation(Bool=$User.bIsFrontEndPreview)
    )
    Control Roll On Collision=true
    Enable Rest State=false
    Friction=0.1
    KillOnCollision=false
    Module.EnableMaxCollisionCount=false
    Module.Randomize Collision Normal Vector=false
    Particle Radius Scale=5
    Radius Calculation Type=::NewEnumerator1
    Restitution=&RandomRangeFloat(Minimum=0.125, Maximum=0.35)
    ScaleRateOfRotation=true
   ) {module=Collision001}
   &Collision(
    Age Colliding Particles=false
    Analytical Collision Plane Space:ENiagaraCoordinateSpace=::Local
    CPU Collision Type=::NewEnumerator3
    Collision Enabled=&Boolean_OrOperation(
     Bool1=&IsPlatformSetActive(New Platform Set=DataInterface() {propertyValues {Platforms {qualityLevelMask=-28}}})
     Bool2=$User.bIsFrontEndPreview
    )
    Control Roll On Collision=true
    Enable Rest State=false
    Friction:NiagaraFloat=&RandomRangeFloat(Minimum=0, Maximum=0.05)
    KillOnCollision=false
    Module.EnableMaxCollisionCount=false
    Module.Randomize Collision Normal Vector=false
    Particle Radius Scale=5
    Perform 2 Analytical Collision Queries=false
    Radius Calculation Type=::NewEnumerator1
    Restitution=&RandomRangeFloat(Minimum=0.125, Maximum=0.35)
    ScaleRateOfRotation=false
   ) {module=Collision002}
   &Drag(
    Drag=&FloatFromCurve(
     FloatCurve=DataInterface() {propertyValues {Curve {keys [<at 0 0 RCIM_Cubic>,<at 0.1 1 RCIM_Cubic>]}}}
     Scale Curve=&RandomRangeFloat(Minimum=0.8, Maximum=1.2)
    )
    Use Linear Drag=true
    Use Rotational Drag=true
   )
   &SolveForcesAndVelocity(Clamp Velocity=false, Limit Acceleration=false)
   &ScaleMeshSize(Scale Factor=&VectorFromFloat(Value=&FloatFromCurve(FloatCurve=DataInterface() {propertyValues {Curve {keys [<at 0 1 RCIM_Cubic>, {arriveTangent=-2.5785, interpMode=RCIM_Cubic, leaveTangent=-2.5785, tangentMode=RCTM_User, time=1, value=0}]}}})))
   &ScaleSpriteSize(Uniform Curve Sprite Scale=DataInterface() {propertyValues {Curve {keys [<at 0 1 RCIM_Cubic>, {arriveTangent=-2.5785, interpMode=RCIM_Cubic, leaveTangent=-2.5785, tangentMode=RCTM_User, time=1, value=0}]}}})
   &DynamicMaterialParameters(
    Index 0 Param 1=2
    Index 0 Param 2=20
    Index 0 Param 3=1
    Index 0 Param 4=1
    Module.Index0Float0Write=true
    Module.Index0Float1Write=true
    Module.Index0Float2Write=true
    Module.Index0Float3Write=true
   )
   /SetVariables(Particles.MeshColor:LinearColor=&Multiply_LinearColorByFloat(
     Linear Color=&MaskLinearColorBySpawnGroup(
      Masked Linear Color=&RandomRangeLinearColor(Minimum=<rgba 0.2537 0.7767 1 1>, Maximum=<rgba 0.0582 0.341 1 1>)
      Passthrough Linear Color=&RandomRangeLinearColor(Minimum=<rgba 0.6479 0.1044 1 1>, Maximum=<rgba 0.8716 0.1772 1 1>)
     )
     Float=2
    )) {isSetParameters=true}
  ]
  rendererClasses [&NiagaraSpriteRendererProperties, &NiagaraMeshRendererProperties]
  renderers [
   {rendererClass=&NiagaraSpriteRendererProperties}
   {rendererClass=&NiagaraMeshRendererProperties, rendererIndex=1}
  ]
 }
 Flash {
  emitterUpdateScript [
   &EmitterState()
   &SpawnBurst_Instantaneous(Use Spawn Probability=false, UseLoopCountLimit=false)
  ]
  particleSpawnScript [
   &InitializeParticle(
    Color Maximum=<rgba 0.7463 0.0677 1 1>
    Color Minimum=<rgba 0.1276 0.5924 1 1>
    Color Mode=::NewEnumerator2
    Lifetime Max=0.24
    Lifetime Min=0.16
    Lifetime Mode=::NewEnumerator1
    Sprite Rotation Mode=::NewEnumerator2
    Sprite Size Max=<56 64>
    Sprite Size Min=<48 56>
    Sprite Size Mode=::NewEnumerator2
    UsePositionOffset=false
    Write Material Random=true
   )
   &SkeletalMeshLocation(
    Bone Sampling Mode=::NewEnumerator4
    MaskSpawnGroup=false
    Sampled Position Offset=<-6 0 0>
    Sampled Position Offset Coordinate Space:ENiagaraSkelMeshTransforms=::NewEnumerator2
    Skeletal Mesh=$System.Skeletal Mesh
   ) {enabled=false}
   &CameraOffset(Camera Offset Amount=10)
   &SubUVAnimation(
    "Sprite Renderer "=DataInterface() {propertyValues {SpriteRenderer=&self:Flash_0.NiagaraSpriteRendererProperties_1}}
    SubUV Animation Mode:ENiagara_SubUVLookupModeV2=::NewEnumerator2
    UseEndFrame=false
    UseStartFrame=false
   )
  ]
  particleUpdateScript [
   &ParticleState()
   &ScaleSpriteSize(
    Uniform Curve Scale=2.25
    Uniform Curve Sprite Scale=DataInterface() {propertyValues {Curve {
       defaultValue=0
       keys [
        {arriveTangent=9.0725, interpMode=RCIM_Cubic, leaveTangent=9.0725, tangentMode=RCTM_User, time=0, value=0}
        {arriveTangent=0.1081, interpMode=RCIM_Cubic, leaveTangent=0.1081, tangentMode=RCTM_User, time=0.1408, value=0.9847}
        {arriveTangent=-3.3038, interpMode=RCIM_Cubic, leaveTangent=-3.3038, tangentMode=RCTM_User, time=1, value=0}
       ]
      }}}
   )
   &ScaleColor(
    Scale Alpha=&FloatFromCurve(FloatCurve=DataInterface() {propertyValues {Curve {
        defaultValue=1
        keys [
         {tangentMode=RCTM_User, time=0, value=1}
         {arriveTangent=-0.0405, interpMode=RCIM_Cubic, leaveTangent=-0.0405, tangentMode=RCTM_User, time=0.7009, value=1}
         {arriveTangent=-8.0089, interpMode=RCIM_Cubic, leaveTangent=-8.0089, tangentMode=RCTM_User, time=1, value=0}
        ]
       }}})
    Scale RGB=&VectorFromFloat(Value=0.4)
    ScaleA=true
    ScaleRGB=true
   )
   &FN_Light_Attributes(
    Color=&RandomRangeLinearColor(Minimum=<rgba 0.1276 0.5924 1 1>, Maximum=<rgba 0.8061 0.2873 1 1>)
    Create Color Attribute=true
    Create Exponent Attribute=true
    Create Position Attribute=true
    Create Radius Attribute=true
    Exponent=4
    Position=&AddVectorToPosition(Position=$Particles.SkeletalMeshLocation.SampledPosition, Vector=<0 25 0>)
    Radius=120
    Scale Color=&FloatFromCurve(
     FloatCurve=DataInterface() {propertyValues {Curve {defaultValue=1, keys [<at 0 1 RCIM_Cubic>, {interpMode=RCIM_Cubic, leaveTangent=-8.0089, time=1, value=0}]}}}
     Scale Curve=3.5
    )
   )
   /SetVariables(
    Particles.ImpactColor:LinearColor=&MakeLinearColorFromVectorAndFloat(
     "Vector (RGB)"=&MakeVectorFromLinearColorRGB(Color=&RandomRangeLinearColor(Minimum=<rgba 0.1276 0.5924 1 1>, Maximum=<rgba 0.7875 0.3057 1 1>))
     "Float (Alpha)"=&FloatFromCurve(
      FloatCurve=DataInterface() {propertyValues {Curve {
         defaultValue=1
         keys [
          {tangentMode=RCTM_User, time=0, value=1}
          {arriveTangent=-0.0405, interpMode=RCIM_Cubic, leaveTangent=-0.0405, tangentMode=RCTM_User, time=0.7009, value=1}
          {arriveTangent=-8.0089, interpMode=RCIM_Cubic, leaveTangent=-8.0089, tangentMode=RCTM_User, time=1, value=0}
         ]
        }}}
      Scale Curve=0.05
     )
    )
    Particles.ImpactSize:Vector2f=&Vector2DFromFloat(Value=&FloatFromCurve(
      FloatCurve=DataInterface() {propertyValues {Curve {
         defaultValue=0
         keys [
          {arriveTangent=9.0725, interpMode=RCIM_Cubic, leaveTangent=9.0725, tangentMode=RCTM_User, time=0, value=0}
          {arriveTangent=0.1081, interpMode=RCIM_Cubic, leaveTangent=0.1081, tangentMode=RCTM_User, time=0.1408, value=0.9847}
          {arriveTangent=-3.3038, interpMode=RCIM_Cubic, leaveTangent=-3.3038, tangentMode=RCTM_User, time=1, value=0}
         ]
        }}}
      Scale Curve=&RandomRangeFloat(Minimum=20, Maximum=25)
     ))
   ) {isSetParameters=true}
   &DynamicMaterialParameters(
    Index 0 Param 1=1
    Module.Index0Float0Write=true
    Module.Index0Float1Write=true
    Module.Index0Float2Write=true
    Module.Index0Float3Write=true
   )
  ]
  rendererClasses [&NiagaraSpriteRendererProperties, &NiagaraLightRendererProperties]
  renderers [
   {rendererClass=&NiagaraSpriteRendererProperties}
   {rendererClass=&NiagaraLightRendererProperties, rendererIndex=1}
   {rendererClass=&NiagaraSpriteRendererProperties, rendererIndex=2}
  ]
 }
 Impact_Streaks {
  emitterUpdateScript [
   &EmitterState()
   &SpawnBurst_Instantaneous(
    Spawn Count=&RandomRangeInt(Minimum=8, Maximum=12)
    Use Spawn Probability=false
    UseLoopCountLimit=false
   )
  ]
  particleSpawnScript [
   &InitializeParticle(
    Color=&SelectLinearColorFromArray(
     Array Sampling Mode=::NewEnumerator0
     Direct Array Index=?
     Color Selection Array=DataInterface() {propertyValues {ColorData [<rgba 0.026 1.3635 5 1>, <rgba 2.4827 0.4431 5 1>, <rgba 5 0.4632 4.8534 1>]}}
     Random Seed=0
    )
    Lifetime Max=0.48
    Lifetime Min=0.24
    Lifetime Mode=::NewEnumerator1
    Sprite Size Max=<7 45>
    Sprite Size Min=<2 15>
    Sprite Size Mode=::NewEnumerator2
    UsePositionOffset=false
    Write Material Random=true
   )
   &SubUVAnimation(End Frame=3, SubUV Animation Mode:ENiagara_SubUVLookupMode=::NewEnumerator2)
   &ShapeLocation(
    MaskSpawnGroup=false
    Override Seed=false
    Ring Radius=3
    Rotation Axis=<0 1 0>
    Rotation Mode=::NewEnumerator1
    Shape Primitive=::NewEnumerator4
    UseOwnerScale=true
   )
   &AddVelocity(
    Module.Use Falloff Distance=false
    Override Seed=false
    Velocity Mode=::NewEnumerator1
    Velocity Speed=&RandomRangeFloat(Minimum=10, Maximum=25)
   )
   &CameraOffset(Camera Offset Amount=30)
  ]
  particleUpdateScript [
   &SolveForcesAndVelocity(Clamp Velocity=false, Limit Acceleration=false)
   &ParticleState()
   &ScaleColor(
    Scale Alpha=&FloatFromCurve(
     FloatCurve=DataInterface() {propertyValues {Curve {keys [{arriveTangent=-3.3102, interpMode=RCIM_Cubic, leaveTangent=-3.3102, tangentMode=RCTM_User, time=0, value=1}, <at 1 0 RCIM_Cubic>]}}}
     Scale Curve=0.075
    )
    Scale RGB=&MakeVectorFromLinearColorRGB(Color=&ColorFromCurve(ColorCurve=DataInterface() {propertyValues {
        AlphaCurve {keys []}
        BlueCurve {keys [<at 0 10 RCIM_Cubic>,<at 0.125 1>,<at 0.6298 0.1878 RCIM_Cubic>]}
        GreenCurve {keys [<at 0 10 RCIM_Cubic>,<at 0.125 1>,<at 0.6298 0.1878 RCIM_Cubic>]}
        RedCurve {keys [<at 0 10 RCIM_Cubic>,<at 0.125 1>,<at 0.6298 0.1878 RCIM_Cubic>]}
       }}))
    ScaleA=true
    ScaleRGB=true
   ) {module=ScaleColor001}
   &ScaleSpriteSize(
    Non-Uniform Curve Sprite Scale=DataInterface() {propertyValues {
      XCurve {keys [
        {arriveTangent=0.9067, interpMode=RCIM_Cubic, leaveTangent=0.9067, tangentMode=RCTM_User, time=0, value=0.8}
        {arriveTangentWeight=0.2034, interpMode=RCIM_Cubic, leaveTangentWeight=0.2034, tangentMode=RCTM_User, time=0.5, value=1}
        {arriveTangent=-3.1509, interpMode=RCIM_Cubic, leaveTangent=-3.1509, tangentMode=RCTM_User, time=1, value=0.1}
       ]}
      YCurve {keys [
        {arriveTangent=3.7985, interpMode=RCIM_Cubic, leaveTangent=3.7985, tangentMode=RCTM_User, time=0, value=0}
        {arriveTangent=2, interpMode=RCIM_Cubic, leaveTangent=2, time=0.5, value=1.6}
        <at 1 2 RCIM_Cubic>
       ]}
     }}
    Scale Sprite Size Mode=::NewEnumerator3
   )
  ]
  rendererClasses [&NiagaraSpriteRendererProperties]
  renderers [{rendererClass=&NiagaraSpriteRendererProperties}]
 }
 Sparks {
  emitterUpdateScript [
   &EmitterState()
   &SpawnBurst_Instantaneous(
    Spawn Count=&RandomRangeInt(Minimum=16, Maximum=36)
    Use Spawn Probability=false
    UseLoopCountLimit=false
   ) {module=SpawnBurst_Instantaneous001}
   &SpawnBurst_Instantaneous(
    Spawn Count=&RandomRangeInt(Minimum=16, Maximum=36)
    Spawn Group=1
    Use Spawn Probability=false
    UseLoopCountLimit=false
   )
  ]
  particleSpawnScript [
   &InitializeParticle(
    Color=&MaskLinearColorBySpawnGroup(
     Masked Linear Color=&RandomRangeLinearColor(Minimum=<rgba 0.2982 0.3308 1 1>, Maximum=<rgba 0.1014 0.1781 1 1>)
     Passthrough Linear Color=&RandomRangeLinearColor(Minimum=<rgba 0.6479 0.1044 1 1>, Maximum=<rgba 0.8716 0.1772 1 1>)
    )
    Lifetime Max=3
    Lifetime Min=0.4
    Lifetime Mode=::NewEnumerator1
    Sprite Size Max=<2.7 2>
    Sprite Size Min=<1.6 1.125>
    Sprite Size Mode=::NewEnumerator2
    UsePositionOffset=false
    Write Material Random=true
   )
   &ShapeLocation(
    MaskSpawnGroup=false
    Offset=&MaskVectorBySpawnGroup(Masked Vector=<5 0 0>, Passthrough Vector=<-5 0 0>)
    Offset Mode=::NewEnumerator0
    Override Seed=false
    Sphere Radius=3
    UseOwnerScale=true
   )
   &SkeletalMeshLocation(
    Bone / Socket Index=&SpawnGroup()
    Bone Sampling Mode=::NewEnumerator4
    MaskSpawnGroup=false
    Sampled Position Offset=&AddVectorToPosition(
     Position=$Particles.ShapeLocation.ShapePosition
     Vector=&MaskVectorBySpawnGroup(Masked Vector=<5 0 0>, Passthrough Vector=<-5 0 0>)
    )
    Sampled Position Offset Coordinate Space:ENiagaraSkelMeshTransforms=::NewEnumerator2
    Skeletal Mesh=$System.Skeletal Mesh
   ) {enabled=false}
   &AddVelocity(
    Cone Angle=180
    Cone Axis=&MaskVectorBySpawnGroup(Masked Vector=<-1 0.5 0>, Passthrough Vector=<1 0.5 0>)
    Module.UseConeFalloff=true
    Module.UseVelDistribution=true
    Override Seed=false
    Speed Falloff From Cone Axis=0.2
    Velocity Mode=::NewEnumerator2
    Velocity Speed=&RandomRangeFloat(Minimum=280, Maximum=780)
   )
  ]
  particleUpdateScript [
   &GravityForce(Gravity=&MakeVector(X=0, Y=0, Z=&RandomRangeFloat(Minimum=-980, Maximum=-2500)))
   &Drag(
    Drag=&FloatFromCurve(
     FloatCurve=DataInterface() {propertyValues {Curve {keys [<at 0 0 RCIM_Cubic>,<at 0.1 1 RCIM_Cubic>]}}}
     Scale Curve=&RandomRangeFloat(Minimum=0.5, Maximum=3)
    )
    Use Linear Drag=true
    Use Rotational Drag=true
   )
   &SolveForcesAndVelocity(Clamp Velocity=false, Limit Acceleration=false)
   &ParticleState()
   &ScaleSpriteSize(Uniform Curve Sprite Scale=DataInterface() {propertyValues {Curve {keys [<at 0 1 RCIM_Cubic>, {arriveTangent=-1, interpMode=RCIM_Cubic, leaveTangent=-1, time=0.2, value=0.3}, <at 1 0 RCIM_Cubic>]}}})
   &ScaleSpriteSizeBySpeed(Max Scale Factor=<0.8 3.4>, Velocity Threshold=250)
   &DynamicMaterialParameters(
    Index 0 Param 1=2
    Index 0 Param 2=20
    Index 0 Param 3=1
    Index 0 Param 4=1
    Module.Index0Float0Write=true
    Module.Index0Float1Write=true
    Module.Index0Float2Write=true
    Module.Index0Float3Write=true
   )
   &ScaleColor(
    Scale RGB=&VectorFromFloat(Value=&FloatFromCurve(
      FloatCurve=DataInterface() {propertyValues {Curve {
         defaultValue=0
         keys [
          {arriveTangent=-1.1382, interpMode=RCIM_Cubic, leaveTangent=-1.1382, tangentMode=RCTM_User, time=0, value=1}
          {arriveTangent=0.0234, interpMode=RCIM_Cubic, leaveTangent=0.0234, tangentMode=RCTM_User, time=1, value=0}
         ]
        }}}
      Scale Curve=&RandomRangeFloat(Minimum=6, Maximum=12)
     ))
    ScaleA=true
    ScaleRGB=true
   )
  ]
  rendererClasses [&NiagaraSpriteRendererProperties]
  renderers [{rendererClass=&NiagaraSpriteRendererProperties}]
 }
}
summary {
 systemName=NS_Emote_CampusSire_Impact
 emitters [
  {emitterName=Sparks, rendererClasses [&NiagaraSpriteRendererProperties]}
  {emitterName=Flash, rendererClasses [&NiagaraSpriteRendererProperties, &NiagaraLightRendererProperties]}
  {emitterName=Debris, rendererClasses [&NiagaraSpriteRendererProperties, &NiagaraMeshRendererProperties]}
  {emitterName=Impact_Streaks, rendererClasses [&NiagaraSpriteRendererProperties]}
 ]
 userVariables [User.bIsFrontEndPreview {type=NiagaraBool}]
}
